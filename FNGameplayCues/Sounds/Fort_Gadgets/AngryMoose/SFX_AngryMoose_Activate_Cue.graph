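# Sound cue for the activation of the AngryMoose gadget in Fortnite.
# A single wave player plays the AngryMoose_End_03 wave directly, with no randomization or mixing.
# It uses super-close player attenuation and the gadgets sound class at reduced volume, and its concurrency override blocks new instances rather than cutting off playing ones.
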SoundCue /FNGameplayCues/Sounds/Fort_Gadgets/AngryMoose/SFX_AngryMoose_Activate_Cue
root=WavePlayer_0
nodes [{id=WavePlayer_0, properties {soundWaveAssetPtr=&/AngryMoose_End_03}}]
properties {
 attenuationSettings=/Game/Sounds/Attenuation/Fort_Player_SuperClose_Att
 bOverrideConcurrency=true
 maxDistance=800
 soundClassObject=/Game/Sounds/FortSoundClassesAndModes/Fort_Gadgets_StW
 volumeMultiplier=0.3
 concurrencyOverrides {resolutionRule=PreventNew}
}
